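# Test gem material, Default Lit, that fakes a refractive faceted interior using a triangle-pattern normal map.
# A triplanar virtual box function produces an interior normal, and the camera ray is refracted by an IOR parameter.
# A remapped dot product approximates total internal reflection.
# Tinted emissive blends skylight environment samples along reflected and refracted directions, alongside an inverse-Fresnel mask and metallic, specular, roughness and base color parameters.
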Material /Game/Global/Materials/MaterialGym/Materials/Gem/Test/M_Gem_v2
graph {
 properties {
  bAutomaticallySetUsageInEditor=false
  bUsedWithMeshDeformer=true
  bUsedWithSkeletalMesh=true
  pixelDepthOffsetMode=PDOM_Legacy
  usedShadingModels=Default Lit
 }
 nodes [
  n1 = NamedRerouteUsage()
  n2 = ScalarParameter() {defaultValue=50, parameterName=InteriorDepth}
  n3 = ScalarParameter() {defaultValue=1, parameterName=InteriorTiling}
  n4 = ScalarParameter() {defaultValue=10, parameterName=InteriorBlendContrast}
  NamedRerouteDeclaration(None=Transform(None=MaterialFunctionCall!(Texture=@n1, Depth=@n2, Tile=@n3, BlendContrast=@n4, SpeedXY, ActorOffset, bUseLocalCameraVector, bUseDitherSingleSample) {materialFunction=/Game/Global/MaterialFunctions/MF_TriplanarVirtualBox}.Result)) {name=InteriorNormal, nodeColor=<rgba 0.1016 1 0 1>}
  n5 = CameraVectorWS()
  n6 = ScalarParameter() {defaultValue=1.5, parameterName=IOR}
  NamedRerouteDeclaration(None=MaterialFunctionCall!(
    Ray Direction=Multiply(A=@n5) {constB=-1}
    Surface Normal=NamedRerouteUsage()
    Refractive Index Origin=Constant() {r=1}
    Refractive Index Target=NamedRerouteDeclaration(None=@n6) {name=IOR, nodeColor=<rgba 0 1 0.9297 1>}
   ) {materialFunction=/Engine/Functions/Engine_MaterialFunctions01/Vectors/Refract}.Vector) {name=RefractionVector, nodeColor=<rgba 1 0 0.4219 1>}
  n7 = Multiply(A=DotProduct(A=NamedRerouteUsage(), B=NamedRerouteUsage())) {constB=-1}
  n8 = Constant() {r=0.73}
  n9 = Constant() {r=0.78}
  NamedRerouteDeclaration(None=LinearInterpolate(Alpha=MaterialFunctionCall(Input=@n7, Input Low=@n8, Input High=@n9) {materialFunction=&RemapValue_0-1}.Result) {constA=0.1, constB=0.9}) {name=TIR, nodeColor=<rgba 0 0.7422 1 1>}
  n10 = TextureSample(UVs=Multiply(A=TextureCoordinate(), B=ScalarParameter() {defaultValue=3, parameterName=NormalTile}), Tex=NamedRerouteUsage()) {samplerType=SAMPLERTYPE_Normal, texture=/Game/Global/Materials/MaterialGym/Materials/TechArtPatterns/Textures/T_Triangles_N}
  NamedRerouteDeclaration(None=Transform(None=NamedRerouteDeclaration(None=@n10.RGB) {name=TangentNormal, nodeColor=<rgba 1 0.6094 0 1>})) {name=WorldSurfaceNormal, nodeColor=<rgba 0 1 0.1328 1>}
  n11 = TextureObjectParameter() {
   parameterName=Param
   samplerType=SAMPLERTYPE_Normal
   texture=&T_Triangles_N
  }
  NamedRerouteDeclaration(None=@n11) {name=NormalTexture, nodeColor=<rgba 1 0 0.75 1>}
  n12 = ScalarParameter() {defaultValue=5, parameterName=ExponentIn}
  n13 = NamedRerouteUsage()
  NamedRerouteDeclaration(None=OneMinus(None=Fresnel(ExponentIn=@n12, Normal=@n13) {baseReflectFraction=0})) {name=FresnelMask, nodeColor=<rgba 1 0.0469 0 1>}
  n14 = VectorParameter() {parameterName=Tint, defaultValue=<rgba 0.151 0.2229 0.2396 1>}
  n15 = NamedRerouteUsage()
  n16 = NamedRerouteUsage()
  n17 = NamedRerouteUsage()
  n18 = NamedRerouteUsage()
  n19 = Constant() {r=1}
  n20 = LinearInterpolate(
   A=SkyLightEnvMapSample(Direction=ReflectionVectorWS(CustomWorldNormal=@n15), Roughness=ScalarParameter() {defaultValue=0.1, parameterName=ReflectionRoughness})
   B=SkyLightEnvMapSample(
    Direction=MaterialFunctionCall(Ray Direction=@n16, Surface Normal=@n17, Refractive Index Origin=@n18, Refractive Index Target=@n19) {materialFunction=&Refract}.Vector
    Roughness=ScalarParameter() {defaultValue=0.2, parameterName=RefractionRoughness}
   )
   Alpha=NamedRerouteUsage()
  )
  NamedRerouteDeclaration(None=Multiply(A=Multiply(A=@n14.RGB, B=@n20), B=NamedRerouteUsage())) {name=Emissive, nodeColor=<rgba 0 0.2734 1 1>}
  NamedRerouteUsage()
  NamedRerouteUsage()
  ScalarParameter() {defaultValue=0.75, parameterName=Metallic}
  ScalarParameter() {defaultValue=1, parameterName=Specular}
  ScalarParameter() {defaultValue=0.1, parameterName=Roughness}
  VectorParameter() {parameterName=Base Color, defaultValue=<rgba 0.4271 0.4271 0.4271 0>}
 ]
}
